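# Sound cue for seagull call one-shots in Starter Island ambience, using outdoor bird attenuation and an exterior ambient bus send.
# The first two quality slots randomly pick among four gull pools: Starter Island, DirtyDocks (modulated quieter), Mountain and Coastal.
# The last two slots share a smaller pool of three Starter Island calls, excluded from branch culling.
# Each pool randomizes its wave under a pitch and volume modulator.
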SoundCue /Game/Sounds/Artemis/POI/Starter_Island/Spot_Bird_Seagulls_Calls_All_Cue
root=QualityLevel_0
nodes [
 {id=Modulator_0, children [Random_2], properties {pitchMax=1.1, volumeMax=1.1, volumeMin=0.9}}
 {id=Modulator_1, children [Random_0], properties {pitchMax=1.1, volumeMax=1.1, volumeMin=0.9}}
 {id=Modulator_2, children [Random_1], properties {pitchMax=1.1, volumeMax=1.1, volumeMin=0.9}}
 {id=Modulator_4, children [Random_4], properties {pitchMax=1.1, volumeMax=0.9, volumeMin=0.8}}
 {id=Modulator_5, children [Random_6], properties {pitchMax=1.1, volumeMax=1.1, volumeMin=0.9}}
 {id=QualityLevel_0, children [Random_5,Random_5,Modulator_2,Modulator_2]}
 {id=Random_0, children [WavePlayer_0,WavePlayer_3,WavePlayer_4,WavePlayer_5,WavePlayer_6,WavePlayer_7], properties {weights [1,1,1,1,1,1]}}
 {id=Random_1, children [WavePlayer_8, WavePlayer_17, WavePlayer_18], properties {bShouldExcludeFromBranchCulling=true, weights [1,1,1]}}
 {
  id=Random_2
  children [WavePlayer_2, WavePlayer_12, WavePlayer_13, WavePlayer_14, WavePlayer_15, WavePlayer_16, WavePlayer_19]
  properties {weights [1,1,1,1,1,1,1]}
 }
 {id=Random_4, children [WavePlayer_1, WavePlayer_9, WavePlayer_10, WavePlayer_11], properties {weights [1,1,1,1]}}
 {id=Random_5, children [Modulator_1,Modulator_4,Modulator_0,Modulator_5], properties {weights [1,1,1,1]}}
 {
  id=Random_6
  children [WavePlayer_24, WavePlayer_25, WavePlayer_26, WavePlayer_27, WavePlayer_28, WavePlayer_29]
  properties {weights [1,1,1,1,1,1]}
 }
 {id=WavePlayer_0, properties {soundWaveAssetPtr=&/Spot_Starter_Island_Seagull_Calls_06}}
 {id=WavePlayer_1, properties {soundWaveAssetPtr=/Game/Sounds/Apollo/Biomes/DirtyDocks/Biome_DirtyDocks_OneShot_Gulls_04}}
 {id=WavePlayer_10, properties {soundWaveAssetPtr=/Game/Sounds/Apollo/Biomes/DirtyDocks/Biome_DirtyDocks_OneShot_Gulls_06}}
 {id=WavePlayer_11, properties {soundWaveAssetPtr=/Game/Sounds/Apollo/Biomes/DirtyDocks/Biome_DirtyDocks_OneShot_Gulls_07}}
 {
  id=WavePlayer_12
  properties {soundWaveAssetPtr=/Game/Sounds/Artemis/Biomes/Mountains/Mountain_Spots/Spot_Mountain_Seagulls_Squarks_Small_Group_02}
 }
 {
  id=WavePlayer_13
  properties {soundWaveAssetPtr=/Game/Sounds/Artemis/Biomes/Mountains/Mountain_Spots/Spot_Mountain_Seagulls_Squarks_Small_Group_03}
 }
 {
  id=WavePlayer_14
  properties {soundWaveAssetPtr=/Game/Sounds/Artemis/Biomes/Mountains/Mountain_Spots/Spot_Mountain_Seagulls_Squarks_Small_Group_04}
 }
 {
  id=WavePlayer_15
  properties {soundWaveAssetPtr=/Game/Sounds/Artemis/Biomes/Mountains/Mountain_Spots/Spot_Mountain_Seagulls_Squarks_Small_Group_05}
 }
 {
  id=WavePlayer_16
  properties {soundWaveAssetPtr=/Game/Sounds/Artemis/Biomes/Mountains/Mountain_Spots/Spot_Mountain_Seagulls_Squarks_Small_Group_06}
 }
 {id=WavePlayer_17, properties {soundWaveAssetPtr=&Spot_Starter_Island_Seagull_Calls_06}}
 {id=WavePlayer_18, properties {soundWaveAssetPtr=&/Spot_Starter_Island_Seagull_Calls_03}}
 {
  id=WavePlayer_19
  properties {soundWaveAssetPtr=/Game/Sounds/Artemis/Biomes/Mountains/Mountain_Spots/Spot_Mountain_Seagulls_Squarks_Small_Group_07}
 }
 {
  id=WavePlayer_2
  properties {soundWaveAssetPtr=/Game/Sounds/Artemis/Biomes/Mountains/Mountain_Spots/Spot_Mountain_Seagulls_Squarks_Small_Group_01}
 }
 {id=WavePlayer_24, properties {soundWaveAssetPtr=/Game/Sounds/Apollo/Biomes/Coastal_Areas/Biome_Coastal_Gulls_OneShot_05}}
 {id=WavePlayer_25, properties {soundWaveAssetPtr=/Game/Sounds/Apollo/Biomes/Coastal_Areas/Biome_Coastal_Gulls_OneShot_01}}
 {id=WavePlayer_26, properties {soundWaveAssetPtr=/Game/Sounds/Apollo/Biomes/Coastal_Areas/Biome_Coastal_Gulls_OneShot_02}}
 {id=WavePlayer_27, properties {soundWaveAssetPtr=/Game/Sounds/Apollo/Biomes/Coastal_Areas/Biome_Coastal_Gulls_OneShot_03}}
 {id=WavePlayer_28, properties {soundWaveAssetPtr=/Game/Sounds/Apollo/Biomes/Coastal_Areas/Biome_Coastal_Gulls_OneShot_04}}
 {id=WavePlayer_29, properties {soundWaveAssetPtr=/Game/Sounds/Apollo/Biomes/Coastal_Areas/Biome_Coastal_Gulls_OneShot_06}}
 {id=WavePlayer_3, properties {soundWaveAssetPtr=&Spot_Starter_Island_Seagull_Calls_03}}
 {id=WavePlayer_4, properties {soundWaveAssetPtr=&/Spot_Starter_Island_Seagull_Calls_04}}
 {id=WavePlayer_5, properties {soundWaveAssetPtr=&/Spot_Starter_Island_Seagull_Calls_05}}
 {id=WavePlayer_6, properties {soundWaveAssetPtr=&/Spot_Starter_Island_Seagull_Calls_07}}
 {id=WavePlayer_7, properties {soundWaveAssetPtr=&/Spot_Starter_Island_Seagull_Calls_08}}
 {id=WavePlayer_8, properties {soundWaveAssetPtr=&Spot_Starter_Island_Seagull_Calls_08}}
 {id=WavePlayer_9, properties {soundWaveAssetPtr=/Game/Sounds/Apollo/Biomes/DirtyDocks/Biome_DirtyDocks_OneShot_Gulls_05}}
]
properties {
 attenuationSettings=&Fort_TileAmbient_OneShot_Bird_Att
 maxDistance=8005
 soundClassObject=&Fort_Ambient_OneShots_Outdoors
 volumeMultiplier=0.301
 busSends [{audioBus=null, minSendLevel=0.6, sendLevel=1, soundSourceBus=&AmbientAudio_Exterior_SourceBus}]
}
