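# Sound cue for one-shot seagull call ambience on the Starter Island point of interest, played quietly with bird attenuation and the ambient one-shot sound class.
# A quality-level root sends its first two slots to a random pick of six call waves and its last two to a three-wave subset.
# Each branch passes through a modulator for volume and pitch variation.
# The three-wave random node is excluded from branch culling.
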SoundCue /Game/Sounds/Artemis/POI/Starter_Island/Spot_Starter_Island_Seagull_Calls_Cue
root=QualityLevel_0
nodes [
 {id=Modulator_1, children [Random_0], properties {pitchMax=1.1, volumeMax=1.1, volumeMin=0.9}}
 {id=Modulator_2, children [Random_1], properties {pitchMax=1.1, volumeMax=1.1, volumeMin=0.9}}
 {id=QualityLevel_0, children [Modulator_1,Modulator_1,Modulator_2,Modulator_2]}
 {id=Random_0, children [WavePlayer_0,WavePlayer_3,WavePlayer_4,WavePlayer_5,WavePlayer_6,WavePlayer_7], properties {weights [1,1,1,1,1,1]}}
 {id=Random_1, children [WavePlayer_8, WavePlayer_17, WavePlayer_18], properties {bShouldExcludeFromBranchCulling=true, weights [1,1,1]}}
 {id=WavePlayer_0, properties {soundWaveAssetPtr=&/Spot_Starter_Island_Seagull_Calls_06}}
 {id=WavePlayer_17, properties {soundWaveAssetPtr=&Spot_Starter_Island_Seagull_Calls_06}}
 {id=WavePlayer_18, properties {soundWaveAssetPtr=&/Spot_Starter_Island_Seagull_Calls_03}}
 {id=WavePlayer_3, properties {soundWaveAssetPtr=&Spot_Starter_Island_Seagull_Calls_03}}
 {id=WavePlayer_4, properties {soundWaveAssetPtr=&/Spot_Starter_Island_Seagull_Calls_04}}
 {id=WavePlayer_5, properties {soundWaveAssetPtr=&/Spot_Starter_Island_Seagull_Calls_05}}
 {id=WavePlayer_6, properties {soundWaveAssetPtr=&/Spot_Starter_Island_Seagull_Calls_07}}
 {id=WavePlayer_7, properties {soundWaveAssetPtr=&/Spot_Starter_Island_Seagull_Calls_08}}
 {id=WavePlayer_8, properties {soundWaveAssetPtr=&Spot_Starter_Island_Seagull_Calls_08}}
]
properties {
 attenuationSettings=&Fort_TileAmbient_OneShot_Bird_Att
 maxDistance=8005
 soundClassObject=&Fort_Ambient_OneShots
 volumeMultiplier=0.2
 preEffectBusSends [{}]
}
